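# M_PillarMannequins is a Nanite-enabled, Default Lit material for the EchoFrost pillar mannequin meshes, with faked lighting in emissive.
# Four texture layers blend through masks derived from the integer U tile of UV channel 1, and the normal is transformed to world space.
# A tinted fake light on downward-facing surfaces and a half-vector specular from the sky atmosphere light feed exposure-compensated emission, over a darkened base color.
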Material /Hera/Sets/EchoFrost/Meshes/PillarMannequins/Materials/M_PillarMannequins
graph {
 properties {
  bAutomaticallySetUsageInEditor=false
  bTangentSpaceNormal=false
  bUseMaterialAttributes=true
  bUsedWithNanite=true
  pixelDepthOffsetMode=PDOM_Legacy
  usedShadingModels=Default Lit
 }
 nodes [
  n1 = Normalize(VectorInput=Add(A=SkyAtmosphereLightDirection(), B=CameraVectorWS()))
  n2 = GetMaterialAttributes(None=NamedRerouteUsage()) {attributeGetTypes [#1]}
  n3 = Multiply(
   A=Power(Base=DotProduct(A=@n1, B=@n2.Normal), Exp=ScalarParameter() {defaultValue=8, parameterName=Fake Specular Power})
   B=ScalarParameter() {defaultValue=0.2, parameterName=Fake Specular Intensity}
  )
  NamedRerouteDeclaration(None=Multiply(A=Saturate(None=@n3), B=VectorParameter() {parameterName=Specular Color, defaultValue=<rgba 0.1525 0.2065 0.5312 1>}.RGB)) {name=Fake Specular, nodeColor=<rgba 0 1 0.6719 1>}
  n4 = TextureSampleParameter2D() {group=Layer1, parameterName=Normals, samplerType=SAMPLERTYPE_Normal, texture=&FlatNormalBlue}
  n5 = TextureSampleParameter2D() {group=Layer1, parameterName=SpecularMasks, samplerType=SAMPLERTYPE_Masks, texture=&T_Fortnite_Default_S}
  n6 = SetMaterialAttributes!(
   MaterialAttributes
   Base Color=TextureSampleParameter2D() {group=Layer1, parameterName=Diffuse, texture=/Game/Global/Textures/Default/Blanks/T_MidgreysRGB}.RGB
   Normal=@n4.RGB
   Specular=@n5.R
   Metallic=@n5.R
   Roughness=@n5.R
  ) {attributeSetTypes [
    #2
    #1
    #3
    #4
    #5
   ]}
  NamedRerouteDeclaration(None=@n6) {name=Layer1, nodeColor=<rgba 1 0 0.4687 1>}
  n7 = TextureSampleParameter2D() {group=Layer2, parameterName=Normals_Texture_2, samplerType=SAMPLERTYPE_Normal, texture=&FlatNormalBlue}
  n8 = TextureSampleParameter2D() {
   group=Layer2
   parameterName=SpecularMasks_2
   samplerType=SAMPLERTYPE_Masks
   texture=&T_Fortnite_Default_S
  }
  n9 = SetMaterialAttributes(
   Base Color=TextureSampleParameter2D() {group=Layer2, parameterName=Diffuse_Texture_2, texture=&T_MidgreysRGB}.RGB
   Normal=@n7.RGB
   Specular=@n8.R
   Metallic=@n8.R
   Roughness=@n8.R
  ) {attributeSetTypes [
    #2
    #1
    #3
    #4
    #5
   ]}
  NamedRerouteDeclaration(None=@n9) {name=Layer2, nodeColor=<rgba 1 0.7954 0.1358 1>}
  n10 = TextureSampleParameter2D() {group=Layer3, parameterName=Normals_Texture_3, samplerType=SAMPLERTYPE_Normal, texture=&FlatNormalBlue}
  n11 = TextureSampleParameter2D() {
   group=Layer3
   parameterName=SpecularMasks_3
   samplerType=SAMPLERTYPE_Masks
   texture=&T_Fortnite_Default_S
  }
  n12 = SetMaterialAttributes(
   Base Color=TextureSampleParameter2D() {group=Layer3, parameterName=Diffuse_Texture_3, texture=&T_MidgreysRGB}.RGB
   Normal=@n10.RGB
   Specular=@n11.R
   Metallic=@n11.R
   Roughness=@n11.R
  ) {attributeSetTypes [
    #2
    #1
    #3
    #4
    #5
   ]}
  NamedRerouteDeclaration(None=@n12) {name=Layer3, nodeColor=<rgba 0.1209 0.4326 1 1>}
  n13 = TextureSampleParameter2D() {group=Layer4, parameterName=Normals_Texture_4, samplerType=SAMPLERTYPE_Normal, texture=&FlatNormalBlue}
  n14 = TextureSampleParameter2D() {
   group=Layer4
   parameterName=SpecularMasks_4
   samplerType=SAMPLERTYPE_Masks
   texture=&T_Fortnite_Default_S
  }
  n15 = SetMaterialAttributes(
   Base Color=TextureSampleParameter2D() {group=Layer4, parameterName=Diffuse_Texture_4, texture=&T_MidgreysRGB}.RGB
   Normal=@n13.RGB
   Specular=@n14.R
   Metallic=@n14.R
   Roughness=@n14.R
  ) {attributeSetTypes [
    #2
    #1
    #3
    #4
    #5
   ]}
  NamedRerouteDeclaration(None=@n15) {name=Layer4, nodeColor=<rgba 0.0496 1 0.2958 1>}
  n16 = Floor(None=ComponentMask(None=TextureCoordinate() {coordinateIndex=1}) {r=true})
  n17 = Saturate(None=@n16)
  n18 = OneMinus(None=@n17)
  n19 = Saturate(None=Subtract(A=@n16))
  n20 = Subtract(A=@n17, B=@n19)
  n21 = Saturate(None=Subtract(A=@n16) {constB=2})
  n22 = Subtract(A=@n19, B=@n21)
  NamedRerouteDeclaration(None=Saturate(None=MaterialFunctionCall(R=@n18, G=@n20, B=@n22, A=@n21) {materialFunction=&AppendMany}.RGBA)) {name=Layer Mask, nodeColor=<rgba 1 0.0168 0 1>}
  n23 = DotProduct(
   A=GetMaterialAttributes(None=NamedRerouteUsage()) {attributeGetTypes [#1]}.Normal
   B=Constant3Vector() {constant=<rgba 0 0 -1 0>}
  )
  n24 = Saturate(None=Add(A=Multiply(A=@n23) {constB=0.5}) {constB=0.5})
  n25 = ScalarParameter() {defaultValue=4, parameterName=Fake Light Power}
  n26 = Multiply(
   A=Multiply(
    A=Power(Base=@n24, Exp=@n25) {constExponent=8}
    B=VectorParameter() {parameterName=Fake Light Color, defaultValue=<rgba 0.9514 0.1503 1 1>}.RGB
   )
   B=ScalarParameter() {defaultValue=2, parameterName=Fake Light Intensity}
  )
  n27 = NamedRerouteUsage()
  n28 = Add(
   A=Multiply(A=@n26, B=GetMaterialAttributes(None=@n27) {attributeGetTypes [#2]}.Base Color)
   B=Multiply(A=@n26, B=ScalarParameter() {parameterName=Fake Light Addition}) {constB=0.05}
  )
  n29 = NamedRerouteUsage()
  NamedRerouteDeclaration(None=EyeAdaptationInverse(LightValueInput=Add(A=@n28, B=@n29), AlphaInput=Constant() {r=0.75}).EyeAdaptationInverse) {name=Emission, nodeColor=<rgba 0 0.3906 1 1>}
  n30 = NamedRerouteUsage()
  n31 = NamedRerouteUsage()
  n32 = ComponentMask(None=NamedRerouteUsage()) {g=true}
  n33 = NamedRerouteUsage()
  n34 = BlendMaterialAttributes(
   A=BlendMaterialAttributes(A=BlendMaterialAttributes(A=@n30, B=@n31, Alpha=@n32), B=NamedRerouteUsage(), Alpha=ComponentMask(None=@n33) {b=true})
   B=NamedRerouteUsage()
   Alpha=ComponentMask(None=NamedRerouteUsage()) {a=true}
  )
  n35 = Transform(None=GetMaterialAttributes(None=@n34) {attributeGetTypes [#1]}.Normal)
  n36 = GetMaterialAttributes(None=NamedRerouteDeclaration(None=SetMaterialAttributes!(MaterialAttributes=@n34, Normal=@n35) {attributeSetTypes [#1]}) {name=Base Material, nodeColor=<rgba 1 0 0.0703 1>}) {attributeGetTypes [#2]}
  SetMaterialAttributes!(
   MaterialAttributes=@n36.MaterialAttributes
   Base Color=Multiply(A=@n36.Base Color, B=ScalarParameter() {defaultValue=0.3, parameterName=Base Color Multiplier}) {constB=0.5}
   Emissive Color=NamedRerouteUsage()
  ) {attributeSetTypes [#2, #6]}
 ]
}
